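# Anime-style material for the EdibleCable biped cosmetic companion, adding a flipbook mouth and optional eye animation to the skeletal mesh.
# Animation-driven smile, frown, jaw-open and emote curves are thresholded into remappable indices for sampling a grid texture with vertex-color masking.
# Face results can feed lit/shaded color, emissive and opacity.
# The shared Anime 2023 shader functions use a companion sphere-normal override, with static switches for UV and vertex-color debug views.
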Material /CosmeticCompanions/Assets/Biped/EdibleCable/Materials/M_SK_Anime_Flipbook_EdibleCable
graph {
 properties {
  bAutomaticallySetUsageInEditor=false
  bUseMaterialAttributes=true
  bUsedWithClothing=true
  bUsedWithMeshDeformer=true
  bUsedWithMorphTargets=true
  bUsedWithNiagaraMeshParticles=true
  bUsedWithSkeletalMesh=true
  bUsedWithStaticLighting=true
  pixelDepthOffsetMode=PDOM_Legacy
  shadingModel=MSM_Cloth
  usedShadingModels=Cloth
 }
 nodes [
  NamedRerouteDeclaration(None=StaticBoolParameter() {group=Flipbook 06 Debug, parameterName=Flipbook.UseDebugVertColor, sortPriority=-1}) {name=UseDebugVertColor, nodeColor=<rgba 0.1953 1 0 1>}
  n1 = VectorParameter() {
   group=Flipbook 04 Mouth
   parameterName=Flipbook.MouthRemap_4-7
   sortPriority=201
   channelNames {a=7 - Unused, b=6 - Unused, g=5 - Unused, r=4 - Emote 01}
   defaultValue=<rgba 4 5 6 7>
  }
  NamedRerouteDeclaration(None=@n1.RGBA) {name=Mouth_Remap_4-7, nodeColor=<rgba 1 0.6652 0.0014 1>}
  NamedRerouteDeclaration(None=ScalarParameter() {group="Flipbook 04.01 Mouth | Curves (Driven By Animation)", parameterName=R_smile_pose, sortPriority=0}) {name=Smile_Pose, nodeColor=<rgba 1 0.2105 0 1>}
  n2 = ScalarParameter() {
   controlType=Enumeration
   enumeration=/Game/Global/Materials/Enums/E_Mat_OnOff
   group=Flipbook 01 Global
   parameterName=Flipbook.ApplyOpacityMask
   sortPriority=4
  }
  NamedRerouteDeclaration(None=LinearInterpolate(A=Constant(), B=NamedRerouteUsage(), Alpha=@n2)) {name=Flipbook_Opacity, nodeColor=<rgba 0 1 0.6016 1>}
  n3 = NamedRerouteUsage()
  n4 = Multiply(A=NamedRerouteUsage(), B=ScalarParameter() {defaultValue=1, group=Flipbook 01 Global, parameterName=Flipbook.EmissiveStrength})
  n5 = NamedRerouteUsage()
  n6 = ScalarParameter() {
   controlType=Enumeration
   enumeration=&E_Mat_OnOff
   group=Flipbook 01 Global
   parameterName=Flipbook.ApplyEmissive
   sortPriority=3
  }
  NamedRerouteDeclaration(None=LinearInterpolate(A=NamedRerouteUsage(), B=LinearInterpolate(A=@n3, B=@n4, Alpha=@n5), Alpha=@n6)) {name=Flipbook_Emissive, nodeColor=<rgba 0.108 0.2905 1 1>}
  NamedRerouteDeclaration(None=Constant()) {name=Index_Neutral, nodeColor=<rgba 0.2818 0.3488 1 1>}
  NamedRerouteDeclaration(None=Constant() {r=3}) {name=Index_Jaw_Open, nodeColor=<rgba 0.2818 0.3488 1 1>}
  NamedRerouteDeclaration(None=Constant() {r=1}) {name=Index_Smile, nodeColor=<rgba 0.2818 0.3488 1 1>}
  n7 = TextureSampleParameter2D() {
   group=Eye Animation 01 Global
   parameterName=EyeAnimation.EyeMask
   samplerSource=SSM_Wrap_WorldGroupSettings
   samplerType=SAMPLERTYPE_Masks
   sortPriority=3
   texture=&T_WhiteMask
  }
  n8 = VertexColor()
  n9 = ScalarParameter() {
   controlType=Enumeration
   enumeration=&E_Mat_OnOff
   group=Eye Animation 01 Global
   parameterName=EyeAnimation.UseVertexColorForEyeMask
   sortPriority=0
  }
  n10 = LinearInterpolate(A=@n7.RGBA, B=AppendVector(A=@n8, B=@n8), Alpha=@n9)
  n11 = TextureCoordinate() {coordinateIndex=1}
  n12 = Sign(None=Subtract(A=ComponentMask(None=@n11) {r=true}) {constB=0.0001})
  n13 = MaterialFunctionCall!(
   Color Lit=NamedRerouteUsage()
   Color Shaded=NamedRerouteUsage()
   Use UV2=StaticBoolParameter() {defaultValue=true, group=Eye Animation 01 Global, parameterName=EyeAnimation.UseUV2, sortPriority=-10}
   Mask=StaticSwitchParameter(
    True=ChannelMaskParameter(None=@n10) {group=Eye Animation 01 Global, parameterName=EyeAnimation.EyeMaskChannel, sortPriority=1}.R
    False=Saturate(None=@n12)
   ) {group=Eye Animation 01 Global, parameterName=EyeAnimation.UseEyeMask, sortPriority=0}
  ) {materialFunction=/CosmeticCompanions/Materials/MaterialFunctions/MF_SK_EyeAnimation}
  NamedRerouteDeclaration(None=StaticSwitch(True=@n13.Shaded, False=NamedRerouteUsage(), Value=NamedRerouteUsage())) {name=EeyAnimColor_Shaded, nodeColor=<rgba 0.125 1 0 1>}
  n14 = MaterialFunctionCall!(Material=NamedRerouteUsage(), Vertex Color Hide Mask) {materialFunction=/Game/Global/Materials/Character_Parent_2023/FlipbookFace_Parent_2023/MaterialFunctions/MF_Flipbook_Debug}
  NamedRerouteDeclaration(None=@n14.Debug Vert Color) {name=DebugVertColor, nodeColor=<rgba 0 0.6953 1 1>}
  NamedRerouteDeclaration(None=@n14.Debug Vert Color WPO) {name=DebugVertColor_WPO, nodeColor=<rgba 1 0.9844 0 1>}
  NamedRerouteDeclaration(None=@n14.Debug UV WPO) {name=DebugUV_WPO, nodeColor=<rgba 1 0 0.3047 1>}
  NamedRerouteDeclaration(None=StaticBoolParameter() {group=Eye Animation 01 Global, parameterName="UseEyeAnimation (Overrides Flipbook Eyes)", sortPriority=-30}) {name=UseEyeAnimation, nodeColor=<rgba 0.1953 1 0 1>}
  NamedRerouteDeclaration(None=ScalarParameter() {group="Flipbook 04.01 Mouth | Curves (Driven By Animation)", parameterName=jaw_open_pose, sortPriority=2}) {name=Jaw_Open_Pose, nodeColor=<rgba 1 0.2105 0 1>}
  NamedRerouteDeclaration(None=ScalarParameter() {group="Flipbook 04.01 Mouth | Curves (Driven By Animation)", parameterName=R_frown_pose, sortPriority=1}) {name=Frown_Pose, nodeColor=<rgba 1 0.2105 0 1>}
  n15 = NamedRerouteUsage()
  n16 = Multiply(A=NamedRerouteUsage(), B=ScalarParameter() {defaultValue=1, parameterName=FlipbookFaceAnimeShadedValue})
  n17 = NamedRerouteUsage()
  n18 = ScalarParameter() {
   controlType=Enumeration
   defaultValue=1
   enumeration=&E_Mat_OnOff
   group=Flipbook 01 Global
   parameterName=Flipbook.ApplyBaseColor
   sortPriority=2
  }
  NamedRerouteDeclaration(None=LinearInterpolate(A=NamedRerouteUsage(), B=LinearInterpolate(A=@n15, B=@n16, Alpha=@n17), Alpha=@n18)) {name=Flipbook_ColorShaded, nodeColor=<rgba 1 0 0.0937 1>}
  n19 = NamedRerouteUsage()
  n20 = NamedRerouteUsage()
  n21 = NamedRerouteUsage()
  n22 = ScalarParameter() {
   controlType=Enumeration
   defaultValue=1
   enumeration=&E_Mat_OnOff
   group=Flipbook 01 Global
   parameterName=Flipbook.ApplyBaseColor
   sortPriority=2
  }
  NamedRerouteDeclaration(None=LinearInterpolate(A=NamedRerouteUsage(), B=LinearInterpolate(A=@n19, B=@n20, Alpha=@n21), Alpha=@n22)) {name=Flipbook_ColorLit, nodeColor=<rgba 0 0.4141 1 1>}
  NamedRerouteDeclaration(None=StaticBoolParameter() {group=Flipbook 06 Debug, parameterName=Flipbook.UseDebugUVs, sortPriority=-1}) {name=UseDebugUVs, nodeColor=<rgba 0.1953 1 0 1>}
  NamedRerouteDeclaration(None=ScalarParameter() {defaultValue=0.1, group=Flipbook 04.02 Mouth | Thresholds, parameterName=Threshold_Frown, sortPriority=1}) {name=Threshold_Frown, nodeColor=<rgba 1 0.0834 0.4399 1>}
  NamedRerouteDeclaration(None=Constant() {r=2}) {name=Index_Frown, nodeColor=<rgba 0.2818 0.3488 1 1>}
  NamedRerouteDeclaration(None=If(A=NamedRerouteUsage(), B=NamedRerouteUsage(), "A > B"=NamedRerouteUsage(), "A < B"=NamedRerouteUsage())) {name=Smile_Output, nodeColor=<rgba 0 0.1532 1 1>}
  NamedRerouteDeclaration(None=If(A=NamedRerouteUsage(), B=NamedRerouteUsage(), "A > B"=NamedRerouteUsage(), "A < B"=NamedRerouteUsage())) {name=Frown_Output, nodeColor=<rgba 0 0.1532 1 1>}
  n23 = MaterialFunctionCall!(CustomUV) {materialFunction=/Game/Global/Materials/Character_Parent_2023/Anime_Parent_2023/MaterialFunctions/MFC_AnimeShader_2023_Pre}
  NamedRerouteDeclaration(None=@n23.Color_Lit) {name=Color_Lit, nodeColor=<rgba 0 1 0.8359 1>}
  n24 = StaticBoolParameter() {
   desc="Needed for when additional areas of the mesh use vertex colors for hiding geo. Avoiding conflict with the filpbooks vertex mask functionality. "
   group=Flipbook 01 Global
   parameterName=Flipbook.UseExtraVertexMask
   sortPriority=-7
  }
  NamedRerouteDeclaration(None=@n24) {name=UseExtraVertexMask, nodeColor=<rgba 0.1953 1 0 1>}
  NamedRerouteDeclaration(None=If(A=NamedRerouteUsage(), B=NamedRerouteUsage(), "A > B"=NamedRerouteUsage(), "A < B"=NamedRerouteUsage())) {name=Jaw_Open_Output, nodeColor=<rgba 0 0.1532 1 1>}
  NamedRerouteDeclaration(None=ScalarParameter() {group="Flipbook 04.01 Mouth | Curves (Driven By Animation)", parameterName=mouth_emote_01, sortPriority=2}) {name=Mouth_Emote_01, nodeColor=<rgba 1 0.2105 0 1>}
  NamedRerouteDeclaration(None=ScalarParameter() {defaultValue=0.1, group=Flipbook 04.02 Mouth | Thresholds, parameterName=Threshold_Mouth_Emote_01, sortPriority=2}) {name=Threshold_Mouth_Emote_01, nodeColor=<rgba 1 0.0834 0.4399 1>}
  NamedRerouteDeclaration(None=@n23.Color_Shaded) {name=Color_Shaded, nodeColor=<rgba 1 0.5156 0 1>}
  n25 = MaterialFunctionCall!(
   Curve Index=NamedRerouteUsage()
   Use Texture=NamedRerouteUsage()
   ----------------------
   Vertex Color=NamedRerouteUsage()
   ---------------------
   Flipbook Texture=NamedRerouteUsage()
   Column Count=NamedRerouteUsage()
   Row Count=NamedRerouteUsage()
   UVs=NamedRerouteUsage()
   UV Offset=NamedRerouteUsage()
   UV Scale=NamedRerouteUsage()
  ) {materialFunction=/Game/Global/Materials/Character_Parent_2023/FlipbookFace_Parent_2023/MaterialFunctions/MF_Sample_Flipbook}
  NamedRerouteDeclaration(None=@n25.Flipbook Vertex Mask) {name=MouthVertexMask, nodeColor=<rgba 0.2092 0.8266 1 1>}
  NamedRerouteDeclaration(None=@n25.Flipbook Color) {name=MouthColor, nodeColor=<rgba 0.2092 0.8266 1 1>}
  NamedRerouteDeclaration(None=@n25.Flipbook Alpha) {name=MouthAlpha, nodeColor=<rgba 0.2092 0.8266 1 1>}
  NamedRerouteDeclaration(None=Constant() {r=4}) {name=Index_Mouth_Emote_01, nodeColor=<rgba 0.2818 0.3488 1 1>}
  NamedRerouteDeclaration(None=StaticSwitch(True=@n13.Lit, False=NamedRerouteUsage(), Value=NamedRerouteUsage())) {name=EyeAnimColor_Lit, nodeColor=<rgba 0.1953 0 1 1>}
  NamedRerouteDeclaration(None=If(A=NamedRerouteUsage(), B=NamedRerouteUsage(), "A > B"=NamedRerouteUsage(), "A < B"=NamedRerouteUsage())) {name=Mouth_Emote_Output, nodeColor=<rgba 0 0.1532 1 1>}
  n26 = VectorParameter() {
   group=Flipbook 04 Mouth
   parameterName=Flipbook.MouthRemap_0-3
   sortPriority=201
   channelNames {a="3 - Jaw Open ", b=2 - Frown, g=1 - Smile, r=0 - Neutral}
   defaultValue=<rgba 0 1 2 3>
  }
  NamedRerouteDeclaration(None=@n26.RGBA) {name=Mouth_Remap_0-3, nodeColor=<rgba 1 0.6652 0.0014 1>}
  NamedRerouteDeclaration(None=MaterialFunctionCall!(Index=NamedRerouteUsage(), Input_0-3=NamedRerouteUsage(), Input_4-7=NamedRerouteUsage()) {materialFunction=/Game/Global/Materials/Character_Parent_2023/FlipbookFace_Parent_2023/MaterialFunctions/MF_RemapIndex_8}.Remapped Index) {name=Mouth_CurveIndex, nodeColor=<rgba 1 0.4453 0 1>}
  NamedRerouteDeclaration(None=ChannelMaskParameter(None=VertexColor()) {group=Flipbook 04 Mouth, parameterName=Flipbook.MouthVertexColor}.R) {name=Mouth_VertexColor, nodeColor=<rgba 1 0.9844 0 1>}
  NamedRerouteDeclaration(None=StaticBoolParameter() {defaultValue=true, group=Flipbook 04 Mouth, parameterName=Flipbook.UseTextureForMouth, sortPriority=-20}) {name=UseTextureForMouth, nodeColor=<rgba 0 1 0.3438 1>}
  n27 = TextureObjectParameter() {
   group=Flipbook 04 Mouth
   parameterName=Flipbook.MouthTexture
   sortPriority=2
   texture=/Game/Global/Materials/Character_Parent_2023/FlipbookFace_Parent_2023/Textures/T_black_emptyalpha
  }
  NamedRerouteDeclaration(None=@n27) {name=Mouth_FlipbookTexture, nodeColor=<rgba 1 0.1875 0 1>}
  NamedRerouteDeclaration(None=ChannelMaskParameter(None=VertexColor()) {group=Flipbook 01 Global, parameterName=Flipbook.ExtraVertexColor, sortPriority=-6}.R) {name=Extra_VertexColor, nodeColor=<rgba 1 0.9844 0 1>}
  NamedRerouteDeclaration(None=ScalarParameter() {defaultValue=3, group=Flipbook 04 Mouth, parameterName=Flipbook.MouthRowCount, sortPriority=3}) {name=Mouth_RowCount, nodeColor=<rgba 1 0 0.7734 1>}
  NamedRerouteDeclaration(None=ScalarParameter() {defaultValue=3, group=Flipbook 04 Mouth, parameterName=Flipbook.MouthColumnCount, sortPriority=4}) {name=Mouth_ColumnCount, nodeColor=<rgba 1 0 0.7734 1>}
  NamedRerouteDeclaration(None=StaticSwitchParameter(True=TextureCoordinate() {coordinateIndex=1}, False=TextureCoordinate()) {group=Flipbook 04 Mouth, parameterName=Flipbook.MouthUseUV2, sortPriority=10}) {name=Mouth_UVs, nodeColor=<rgba 1 0.5859 0 1>}
  n28 = ScalarParameter() {defaultValue=1, group=Flipbook 04 Mouth, parameterName=Flipbook.MouthUVScaleX, sortPriority=12}
  n29 = ScalarParameter() {defaultValue=1, group=Flipbook 04 Mouth, parameterName=Flipbook.MouthUVScaleY, sortPriority=13}
  NamedRerouteDeclaration(None=Multiply(
    A=ScalarParameter() {defaultValue=1, group=Flipbook 04 Mouth, parameterName=Flipbook.MouthGlobalScale, sortPriority=12}
    B=AppendVector(A=@n28, B=@n29)
   )) {name=Mouth_UVScale, nodeColor=<rgba 0.2656 0 1 1>}
  NamedRerouteDeclaration(None=AppendVector(
    A=ScalarParameter() {group=Flipbook 04 Mouth, parameterName=Flipbook.MouthUVOffsetX, sortPriority=20}
    B=ScalarParameter() {group=Flipbook 04 Mouth, parameterName=Flipbook.MouthUVOffsetY, sortPriority=21}
   )) {name=Mouth_UVOffset, nodeColor=<rgba 0.0547 1 0 1>}
  NamedRerouteDeclaration(None=Saturate(None=NamedRerouteUsage())) {name=Flipbook_FaceMask, nodeColor=<rgba 1 0.7969 0 1>}
  NamedRerouteDeclaration(None=LinearInterpolate(B=NamedRerouteUsage(), Alpha=NamedRerouteUsage())) {name=Flipbook_FaceColor, nodeColor=<rgba 1 0.849 0.0986 1>}
  NamedRerouteDeclaration(None=ScalarParameter() {defaultValue=0.1, group=Flipbook 04.02 Mouth | Thresholds, parameterName=Threshold_Smile, sortPriority=0}) {name=Threshold_Smile, nodeColor=<rgba 1 0.0834 0.4399 1>}
  NamedRerouteDeclaration(None=ScalarParameter() {defaultValue=0.1, group=Flipbook 04.02 Mouth | Thresholds, parameterName=Threshold_Jaw_Open, sortPriority=2}) {name=Threshold_Jaw_Open, nodeColor=<rgba 1 0.0834 0.4399 1>}
  NamedRerouteDeclaration(None=@n23.Emissive) {name=Emissive, nodeColor=<rgba 1 0.1875 0 1>}
  n30 = MaterialFunctionCall!(
   CustomLightVector
   OverrideNormal_WS=MaterialFunctionCall!(Normal=@n23.Normal) {materialFunction=/CosmeticCompanions/Materials/MaterialFunctions/MF_CompanionAnime_NormalSphere}.Normal
   STT=@n23.STT
   Normal_Map=@n23.STT
   DFL=@n23.STT
   FX=@n23.STT
  ) {materialFunction=/Game/Global/Materials/Character_Parent_2023/Anime_Parent_2023/MaterialFunctions/MFC_AnimeShader_2023_Calculations}
  n31 = GetMaterialAttributes(None=MaterialFunctionCall!(
    Shading_Gradient=@n30.Shading_Gradient
    Threshold_Gradient=@n30.Shading_Gradient
    Specular_Gradient=@n30.Shading_Gradient
    Sheen_Mask=@n30.Shading_Gradient
    Thickness=@n30.Shading_Gradient
    DFL_Mask=@n30.Shading_Gradient
    FX=@n30.Shading_Gradient
    Metallic_Gradient=@n30.Shading_Gradient
    RimLight_Gradient=@n30.Shading_Gradient
    NormalBuffer=@n30.Shading_Gradient
    --------
    CINE_LightSources=@n30.CINE_LightSources
    CINE_LightColor=@n30.CINE_LightColor
    ---------
    Color_Lit=NamedRerouteUsage()
    Color_Shaded=NamedRerouteUsage()
    SRM=@n23.SRM
    Emissive=NamedRerouteUsage()
    DFL_Color=@n23.DFL_Color
    Specular_Color=@n23.Specular_Color
    RimLight_Color=@n23.StyleRimLight_Color
   ) {materialFunction=/Game/Global/Materials/Character_Parent_2023/Anime_Parent_2023/MaterialFunctions/MFC_AnimeShader_2023_Post}.Result) {attributeGetTypes [#1, #2]}
  n32 = NamedRerouteUsage()
  n33 = NamedRerouteUsage()
  SetMaterialAttributes!(
   MaterialAttributes=NamedRerouteDeclaration(None=@n31.MaterialAttributes) {name=Material, nodeColor=<rgba 0.6406 0 1 1>}
   Base Color=StaticSwitch(True=NamedRerouteUsage(), False=@n31.Base Color, Value=NamedRerouteUsage())
   Opacity Mask=NamedRerouteUsage()
   World Position Offset=StaticSwitch(True=NamedRerouteUsage(), False=StaticSwitch(True=@n32, False=@n31.World Position Offset, Value=@n33), Value=NamedRerouteUsage())
  ) {attributeSetTypes [#1, #3, #2]}
  TemporalResponsivenessOutput!(None=LinearInterpolate(B=ScalarParameter() {parameterName=Temporal Responsiveness}, Alpha=NamedRerouteUsage()) {constA=0.4})
 ]
}
